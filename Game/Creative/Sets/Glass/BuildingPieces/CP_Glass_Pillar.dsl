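; Creative glass pillar building piece with a player-selectable Color option that changes its material.
; On BeginPlay, option changes and option loads, where cosmetic events can run, the Color value indexes an array of soft material references.
; The chosen material loads asynchronously, then is set on mesh slot 0 and stored as the current material.
; It is also targetable by linear movers and disallows invisible props.
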(event EventBeginPlay
  (Networking|CanExecuteCosmeticEvents
    (:True
      (CallFunction|UpdateMeshVisualsonClient))))

(event OnAnyPropertyChanged(ToyOptionsComponent)
  (Networking|CanExecuteCosmeticEvents
    (:True
      (CallFunction|UpdateMeshVisualsonClient))))

(event OnPropertiesLoaded(ToyOptionsComponent)
  (Networking|CanExecuteCosmeticEvents
    (:True
      (CallFunction|UpdateMeshVisualsonClient))))

(event Custom|UpdateMeshVisualsOnClient
  (bind _color (Variables|UserOptions|Basic|GetColor))
  (bind _returnvalue (Math|Conversions|ToInteger(Byte) _color))
  (bind _meshvisiblematerials (Variables|Default|GetMeshVisibleMaterials))
  (bind _output (Utilities|Array|Get(acopy) _meshvisiblematerials _returnvalue))
  (if (Utilities|IsValidSoftObjectReference _output)
    (bind _object (Utilities|AsyncLoadAsset _output)
      (:Completed
        (CallFunction|MeshMaterialLoaded _object)))))

(event Custom|MeshMaterialLoaded
  (params (Object :type Object%20Reference))
  (bind _asmaterial_interface (Utilities|Casting|CastToMaterialInterface Object))
  (Rendering|Material|SetMaterial (Variables|Debug|GetStaticMeshComponent) 0 _asmaterial_interface)
  (Variables|Default|SetCurrentMeshMaterial _asmaterial_interface))

(fn IsTargetableByLinearMover
  (return true))

(fn ShouldInvisiblePropsBeAllowed
  (return false))

(fn ConstructionScript
  (|CallParentFunction))
